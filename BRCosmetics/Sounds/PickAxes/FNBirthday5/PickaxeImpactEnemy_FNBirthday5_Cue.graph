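# Sound cue for enemy impacts of the FNBirthday5 cosmetic pickaxe, using pickaxe impact attenuation and the impacts sound class.
# The root makes a weighted random pick, with replacement, favoring a plain impact over a rarer party-layered version.
# The plain branch picks one of six impact waves.
# The layered branch mixes a random party layer, at slightly reduced volume, with a random impact from the same six.
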SoundCue /BRCosmetics/Sounds/PickAxes/FNBirthday5/PickaxeImpactEnemy_FNBirthday5_Cue
root=Random_5
nodes [
 {id=Mixer_2, children [Random_3,Random_4], properties {inputVolume [0.8,1]}}
 {id=Random_0, children [WavePlayer_0,WavePlayer_1,WavePlayer_2,WavePlayer_3,WavePlayer_4,WavePlayer_5], properties {weights [1,1,1,1,1,1]}}
 {
  id=Random_3
  children [WavePlayer_13, WavePlayer_14, WavePlayer_15, WavePlayer_16, WavePlayer_17, WavePlayer_18]
  properties {weights [1,1,1,1,1,1]}
 }
 {
  id=Random_4
  children [WavePlayer_25, WavePlayer_26, WavePlayer_27, WavePlayer_28, WavePlayer_29, WavePlayer_30]
  properties {weights [1,1,1,1,1,1]}
 }
 {id=Random_5, children [Random_0,Mixer_2], properties {bRandomizeWithoutReplacement=false, weights [1,0.4]}}
 {id=WavePlayer_0, properties {soundWaveAssetPtr=&/Pickaxe_FNBirthday5_Impact_01}}
 {id=WavePlayer_1, properties {soundWaveAssetPtr=&/Pickaxe_FNBirthday5_Impact_02}}
 {id=WavePlayer_13, properties {soundWaveAssetPtr=&/Pickaxe_FNBirthday5_Impact_Party_Lyr01}}
 {id=WavePlayer_14, properties {soundWaveAssetPtr=&/Pickaxe_FNBirthday5_Impact_Party_Lyr02}}
 {id=WavePlayer_15, properties {soundWaveAssetPtr=&/Pickaxe_FNBirthday5_Impact_Party_Lyr03}}
 {id=WavePlayer_16, properties {soundWaveAssetPtr=&/Pickaxe_FNBirthday5_Impact_Party_Lyr04}}
 {id=WavePlayer_17, properties {soundWaveAssetPtr=&/Pickaxe_FNBirthday5_Impact_Party_Lyr05}}
 {id=WavePlayer_18, properties {soundWaveAssetPtr=&/Pickaxe_FNBirthday5_Impact_Party_Lyr06}}
 {id=WavePlayer_2, properties {soundWaveAssetPtr=&/Pickaxe_FNBirthday5_Impact_03}}
 {id=WavePlayer_25, properties {soundWaveAssetPtr=&Pickaxe_FNBirthday5_Impact_01}}
 {id=WavePlayer_26, properties {soundWaveAssetPtr=&Pickaxe_FNBirthday5_Impact_02}}
 {id=WavePlayer_27, properties {soundWaveAssetPtr=&Pickaxe_FNBirthday5_Impact_03}}
 {id=WavePlayer_28, properties {soundWaveAssetPtr=&/Pickaxe_FNBirthday5_Impact_04}}
 {id=WavePlayer_29, properties {soundWaveAssetPtr=&/Pickaxe_FNBirthday5_Impact_05}}
 {id=WavePlayer_3, properties {soundWaveAssetPtr=&Pickaxe_FNBirthday5_Impact_04}}
 {id=WavePlayer_30, properties {soundWaveAssetPtr=&/Pickaxe_FNBirthday5_Impact_06}}
 {id=WavePlayer_4, properties {soundWaveAssetPtr=&Pickaxe_FNBirthday5_Impact_05}}
 {id=WavePlayer_5, properties {soundWaveAssetPtr=&Pickaxe_FNBirthday5_Impact_06}}
]
properties {
 attenuationSettings=/Game/Sounds/Attenuation/Fort_Pickaxe_Impact_Att
 audioPropertiesSheet=/BRCosmetics/Sounds/Settings/PropertySheets/Pickaxe/APS_Cosmetic_Pickaxe_Impact
 maxDistance=4300
 soundClassObject=&Fort_Impacts
 volumeMultiplier=0.735
 concurrencySet [null]
}
